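# Sound cue for looping diegetic Winterfest music at the Helios WinterIsland POI, attenuated as an outdoor speaker with a cone.
# A looping wave player feeds an Enveloper whose volume curve ramps from silence to full.
# The base submix is disabled, and the cue sends to MarinaTown's music party audio bus and outdoor speaker source bus.
# It uses the POI music sound class and overrides concurrency to cap simultaneous instances.
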SoundCue /Game/Sounds/Helios/POI/WinterIsland/Cue_Diegetic_MS_Helios_WinterIsland01
root=Enveloper_0
nodes [
 {
  id=Enveloper_0
  children [WavePlayer_0]
  properties {pitchCurve {editorCurveData {keys [{time=0}]}}, volumeCurve {editorCurveData {keys [<at 0 0>,<at 1 1>]}}}
 }
 {id=WavePlayer_0, properties {bLooping=true, soundWaveAssetPtr=&/S28_Winterfest_Diegetic_Loop_01}}
]
properties {
 attenuationSettings=&/Helios_Music_WinterIsland_OutdoorSpeaker_Cone_Att
 bEnableBaseSubmix=false
 bOverrideConcurrency=true
 maxDistance=5435.2158
 soundClassObject=/Game/Sounds/FortSoundClassesAndModes/Fort_Music_POI_PSM
 volumeMultiplier=1.5
 busSends [{
   audioBus=/Game/Sounds/Helios/POI/MarinaTown/MusicParty_MS_Helios_MarinaTown
   minSendLevel=0
   sendLevel=1
   soundSourceBus=/Game/Sounds/Helios/POI/MarinaTown/Helios_MarinaTown_OutdoorSpeaker_SrcBus
  }]
 concurrencyOverrides {maxCount=4}
}
