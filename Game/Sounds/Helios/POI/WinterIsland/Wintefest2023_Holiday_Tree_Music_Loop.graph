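# Sound cue that loops diegetic holiday tree music for the Winterfest 2023 Winter Island POI, judging by its name and path.
# A single looping wave player feeds the output at reduced volume through the ambient loot loop sound class.
# Its own attenuation override applies logarithmic falloff, a non-spatialized near radius, stereo spread and interpolated listener focus that favors the faced source.
# It shares the ambient loot concurrency set.
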SoundCue /Game/Sounds/Helios/POI/WinterIsland/Wintefest2023_Holiday_Tree_Music_Loop
root=WavePlayer_0
nodes [{id=WavePlayer_0, properties {bLooping=true, soundWaveAssetPtr=&/s28_Winterfest_Diegetic_Loop_03}}]
properties {
 bOverrideAttenuation=true
 maxDistance=1312
 soundClassObject=/Game/Sounds/FortSoundClassesAndModes/Fort_Loot_Ambient_Loop
 volumeMultiplier=0.5
 attenuationOverrides {
  bEnableFocusInterpolation=true
  bEnableListenerFocus=true
  distanceAlgorithm=Logarithmic
  falloffDistance=800
  focusAttackInterpSpeed=0.5
  focusAzimuth=20
  focusReleaseInterpSpeed=1.5
  focusVolumeAttenuation=1.2
  lPFFrequencyAtMax=10000
  lPFRadiusMax=1000
  lPFRadiusMin=500
  nonFocusAzimuth=40
  nonFocusVolumeAttenuation=0.5
  nonSpatializedRadiusStart=200
  occlusionInterpolationTime=0.3
  occlusionLowPassFilterFrequency=3000
  occlusionVolumeAttenuation=0.75
  stereoSpread=512
  attenuationShapeExtents {x=512}
  customAttenuationCurve {editorCurveData {keys [<at 0 1 RCIM_Cubic>, {arriveTangent=-1, interpMode=RCIM_Cubic, leaveTangent=-1, time=0.5, value=0.1}, <at 1 0 RCIM_Cubic>]}}
 }
 concurrencyOverrides {maxCount=1}
 concurrencySet [/Game/Sounds/Concurrency_Rules/Loot_Ambient_Conc]
}
